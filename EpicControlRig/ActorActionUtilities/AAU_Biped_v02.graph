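# Biped editor utility that performs IK/FK matching and switching on a Modular Rig limb module in Sequencer.
# Match Limb finds a rig with selected controls and an active section, then reads the module's Ik Fk Switch at the current frame.
# It runs the module's To FK or To IK event, keys the flipped switch, selects the corresponding control and runs Key Controls.
# Select Control clears selection across the sequence's rigs before selecting one control.
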EditorUtilityBlueprint /EpicControlRig/ActorActionUtilities/AAU_Biped_v02
graphs [
 EventGraph {links [], nodes []}
 Match Limb {nodes [
   n1 = Match Limb!() -> (then, InModule:Modular Rig Object Reference) {is=title K2Node_FunctionEntry}
   n2 = GetObjectName(Object=@n1)
   > Set Module Name(Module Name:Name=String To Name(InString=@n2)) -> (Output_Get:Name=None)
   > Set IK FK Switch(IK FK Switch:Name=String To Name(InString=Append(A=@n2, B=/Ik Fk Switch))) -> (Output_Get:Name=None)
   > n3 = GetFocusedLevelSequence!(execute) -> (then, ReturnValue:Level Sequence Object Reference)
   > n4 = GetControlRigs!(execute, LevelSequence:Level Sequence Object Reference=@n3) -> (then, ReturnValue:Array of Control Rig Sequencer Binding Proxy Structures)
   > n5 = For Each Loop(Array:Array of Control Rig Sequencer Binding Proxy Structures=@n4) -> (Array Element:Control Rig Sequencer Binding Proxy Structure)
   n6 = Cast To ModularRig!(execute=@n5.Completed, Object=Get Control Rig() -> (Control Rig:Control Rig Object Reference)) -> (then, CastFailed, AsModular Rig:Modular Rig Object Reference) {is=title K2Node_DynamicCast}
   > Set Modular Rig(Modular Rig:Modular Rig Object Reference=@n6) -> (Output_Get:Modular Rig Object Reference)
   > n7 = Get Current Local Time!(execute, TimeUnit:EMovieSceneTimeUnit Enum=DisplayRate) -> (
    then
    ReturnValue_Frame_FrameNumber:Frame Number Structure
    ReturnValue_Frame_SubFrame:float=~
    ReturnValue_Time:float=~
    ReturnValue_MarkedFrame:String
    ReturnValue_Timecode:Timecode Structure
    ReturnValue_PositionType:EMovieScenePositionType Enum
    ReturnValue_UpdateMethod:EUpdatePositionMethod Enum
    ReturnValue_bHasJumped:bool=~
   ) {is=title K2Node_CallFunction}
   > n8 = GetLocalControlRigBool!(
    execute
    LevelSequence:Level Sequence Object Reference=@n3
    ControlRig:Control Rig Object Reference=Get Control Rig()
    ControlName:Name=Get IK FK Switch() -> (IK FK Switch:Name=None)
    Frame:Frame Number Structure=@n7
    TimeUnit:EMovieSceneTimeUnit Enum=DisplayRate
   ) -> (then, ReturnValue:bool)
   > n9 = Branch(Condition=@n8)
   n10 = Get Module Name() -> (Module Name:Name=None)
   n11 = Get Modular Rig() -> (Modular Rig:Modular Rig Object Reference)
   > Execute Event On Module By Name!(execute, self:Modular Rig Object Reference=@n11, InEvent:Name=To FK, InModuleName:Name=@n10) -> (then, ReturnValue:bool)
   > Execute!(execute, self:Rig VMHost Object Reference=@n11, InEventName:"Name (by ref)"=MakeLiteralName(Value=Forwards Solve)) -> (then, ReturnValue:bool)
   > n12 = Get Current Local Time(TimeUnit=DisplayRate) -> (ReturnValue_Frame_SubFrame=~, ReturnValue_Time=~, ReturnValue_bHasJumped=~)
   > n13 = GetFocusedLevelSequence()
   > SetInteraction!(execute, bIsInteracting:bool=true) -> (then)
   n14 = Get IK FK Switch() -> (IK FK Switch=None)
   > SetLocalControlRigBool!(
    execute
    LevelSequence:Level Sequence Object Reference=@n13
    ControlRig:Control Rig Object Reference=Get Control Rig()
    ControlName:Name=@n14
    Frame:Frame Number Structure=@n12
    Value:bool=false
    TimeUnit:EMovieSceneTimeUnit Enum=DisplayRate
    bSetKey:bool=true
   ) -> (then)
   > SetInteraction(bIsInteracting=false)
   n15 = GetObjectName(Object=Get InModule() -> (InModule:Modular Rig Object Reference))
   > n16 = Select Control!(
    execute
    self
    Control Rig:Control Rig Object Reference=Get Control Rig()
    Control Name:"Name (by ref)"=String To Name(InString=Append(A=@n15, B=/FK 2))
   ) -> (then)
   n17 = Get Modular Rig()
   Execute Event On Module By Name(execute=[@n21.then, @n16.then], self=@n17, InEvent=Key Controls, InModuleName=Get Module Name() -> (Module Name=None))
   > Execute(self=@n17, InEventName=MakeLiteralName(Value=Forwards Solve))
   > Return Node()
   Execute Event On Module By Name(execute=@n9.else, self=@n11, InEvent=To IK, InModuleName=@n10)
   > Execute(self=@n11, InEventName=MakeLiteralName(Value=Forwards Solve))
   > n18 = Get Current Local Time(TimeUnit=DisplayRate) -> (ReturnValue_Frame_SubFrame=~, ReturnValue_Time=~, ReturnValue_bHasJumped=~)
   > n19 = GetFocusedLevelSequence()
   > SetInteraction(bIsInteracting=true)
   > SetLocalControlRigBool(LevelSequence=@n19, ControlRig=Get Control Rig(), ControlName=@n14, Frame=@n18, Value=true, TimeUnit=DisplayRate, bSetKey=true)
   > SetInteraction(bIsInteracting=false)
   n20 = GetObjectName(Object=Get InModule())
   > n21 = Select Control(Control Rig=Get Control Rig(), Control Name=String To Name(InString=Append(A=@n20, B=/IK)))
   n22 = Break Control Rig Sequencer Binding Proxy!(ControlRigSequencerBindingProxy:"Control Rig Sequencer Binding Proxy Structure (by ref)"=@n5) -> (
    Proxy:Movie Scene Binding Proxy Structure
    ControlRig:Control Rig Object Reference
    Track:Movie Scene Control Rig Parameter Track Object Reference
   ) {is=title K2Node_BreakStruct}
   n23 = GetSections!(execute=@n5.LoopBody, Track:Movie Scene Track Object Reference=@n22.Track) -> (then, ReturnValue:Array of Movie Scene Section Object References)
   > n24 = For Each Loop with Break!(Exec:Exec, Array:Array of Movie Scene Section Object References=@n23, Break:Exec=@n25.then) -> (LoopBody:Exec, Array Element:Movie Scene Section Object Reference, Array Index:int, Completed:Exec)
   > Branch(Condition=IsActive!(self:Movie Scene Section Object Reference=@n24) -> (ReturnValue:bool))
   > Branch(Condition=Is Not Empty!(TargetArray:Array of Names=CurrentControlSelection!(self:Control Rig Object Reference=@n22.ControlRig) -> (ReturnValue:Array of Names)) -> (ReturnValue:bool))
   > n25 = Set Control Rig(Control Rig:Control Rig Object Reference=@n22.ControlRig) -> (Output_Get:Control Rig Object Reference)
   Get Module Name() -> (Module Name=None)
  ]}
 Select Control {nodes [
   n1 = Select Control!() -> (then, Control Rig:Control Rig Object Reference, Control Name:"Name (by ref)") {is=title K2Node_FunctionEntry}
   > SelectNothing(self:Editor Actor Subsystem Object Reference=Get EditorActorSubsystem())
   > n2 = GetFocusedLevelSequence()
   > n3 = GetControlRigs(LevelSequence=@n2)
   > n4 = For Each Loop!(Exec, Array=@n3) -> (
    LoopBody
    Array Element_Proxy:Movie Scene Binding Proxy Structure
    Array Element_ControlRig:Control Rig Object Reference
    Array Element_Track:Movie Scene Control Rig Parameter Track Object Reference
    Array Index
    Completed
   )
   SelectControl!(
    execute=@n4.Completed
    self:Control Rig Object Reference=@n1
    InControlName:"Name (by ref)"=@n1.Control Name
    bSelect:bool=true
    bSetupUndo:bool=false
   ) -> (then)
   ClearControlSelection!(execute=@n4.LoopBody, self:Control Rig Object Reference=@n4.Array Element_ControlRig, bSetupUndo:bool=false) -> (then, ReturnValue:bool)
  ]}
]
variables [
 Control Rig {type=ControlRig, default=null}
 Modular Rig {type=ModularRig, default=null}
]
